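# Fort_Nexus_Transition is a SoundClass, apparently for Nexus transition sounds in the Figment_S08 DirtShell content.
# It is parented to the Fort_SFX_Nexus sound class.
# A passive sound mix modifier with a zero minimum volume threshold ties it to the Fort_Nexus_MixMod mix, so sounds playing in this class automatically trigger that mix.
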SoundClass /Figment_S08/DirtShell_8_5_1/Sounds/FortSoundClassesAndModes/Fort_Nexus_Transition
parent=&/Fort_SFX_Nexus
properties {passiveSoundMixModifiers [{minVolumeThreshold=0, soundMix=/Game/Sounds/FortSoundClassesAndModes/Fort_Nexus_MixMod}]}
