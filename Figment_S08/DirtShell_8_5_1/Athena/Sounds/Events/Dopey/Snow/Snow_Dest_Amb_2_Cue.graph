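# Sound cue that plays a looping snow ambience (Snow_Dest_Amb_03_LOOP) for the Dopey Snow event sounds.
# A looping wave player feeds a smoothed Doppler node at the root, with boosted volume on the Fort_Nexus_Transition sound class.
# Its own attenuation override adds distance falloff, wide stereo spread, distance-driven reverb send and interpolated listener focus that makes the loop louder when faced and quieter off-axis.
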SoundCue /Figment_S08/DirtShell_8_5_1/Athena/Sounds/Events/Dopey/Snow/Snow_Dest_Amb_2_Cue
root=Doppler_0
nodes [
 {id=Doppler_0, children [WavePlayer_0], properties {bUseSmoothing=true}}
 {
  id=WavePlayer_0
  properties {bLooping=true, soundWaveAssetPtr=&/Snow_Dest_Amb_03_LOOP}
 }
]
properties {
 bOverrideAttenuation=true
 maxDistance=1506.2095
 soundClassObject=/Figment_S08/DirtShell_8_5_1/Sounds/FortSoundClassesAndModes/Fort_Nexus_Transition
 volumeMultiplier=1.5
 attenuationOverrides {
  bEnableFocusInterpolation=true
  bEnableListenerFocus=true
  falloffDistance=810.2751
  focusAttackInterpSpeed=5
  focusAzimuth=20
  focusReleaseInterpSpeed=0.5
  focusVolumeAttenuation=2
  nonFocusAzimuth=45
  nonFocusVolumeAttenuation=0.3
  reverbDistanceMax=3000
  reverbDistanceMin=1000
  reverbWetLevelMax=10
  reverbWetLevelMin=0
  stereoSpread=1024
  attenuationShapeExtents {x=695.9344}
 }
}
