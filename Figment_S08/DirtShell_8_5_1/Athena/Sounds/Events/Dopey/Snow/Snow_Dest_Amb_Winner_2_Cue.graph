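# Sound cue for the winner ambience of the Snow destination event, judging by its name, in the Dopey event audio set.
# A single wave player feeds the Snow_End sound straight to the output with boosted volume.
# It routes through the Fort_Nexus_Transition sound class.
# Attenuation overrides turn off distance falloff and spatialization, so it plays as non-positional audio.
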SoundCue /Figment_S08/DirtShell_8_5_1/Athena/Sounds/Events/Dopey/Snow/Snow_Dest_Amb_Winner_2_Cue
root=WavePlayer_2
nodes [{id=WavePlayer_2, properties {soundWaveAssetPtr=&/Snow_End}}]
properties {
 maxDistance=2097152
 soundClassObject=/Figment_S08/DirtShell_8_5_1/Sounds/FortSoundClassesAndModes/Fort_Nexus_Transition
 volumeMultiplier=1.5
 attenuationOverrides {
  bAttenuate=false
  bEnableFocusInterpolation=true
  bSpatialize=false
  falloffDistance=10000
  focusAttackInterpSpeed=5
  focusAzimuth=20
  focusReleaseInterpSpeed=0.5
  focusVolumeAttenuation=2
  nonFocusAzimuth=45
  nonFocusVolumeAttenuation=0.3
  reverbDistanceMax=3000
  reverbDistanceMin=1000
  reverbWetLevelMax=10
  reverbWetLevelMin=0
  stereoSpread=1024
  attenuationShapeExtents {x=695.9344}
 }
}
