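# Sound cue named for a resonant snow impact in the Dopey event audio, with no sound wave connected in this file.
# Its root is a single Modulator node with an empty input, so the graph as written has no audio source.
# It routes to the Fort_Nexus_Transition sound class and uses the Snow_Impacts_2_Att attenuation asset.
# A concurrency override caps simultaneous instances at two, blocking new ones beyond that.
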SoundCue /Figment_S08/DirtShell_8_5_1/Athena/Sounds/Events/Dopey/Snow/Snow_Impact_Resonant
root=Modulator_0
nodes [{id=Modulator_0, children [null]}]
properties {
 attenuationSettings=&/Snow_Impacts_2_Att
 bOverrideConcurrency=true
 maxDistance=6000
 soundClassObject=/Figment_S08/DirtShell_8_5_1/Sounds/FortSoundClassesAndModes/Fort_Nexus_Transition
 volumeMultiplier=1
 attenuationOverrides {falloffDistance=3000, attenuationShapeExtents {x=0}}
 concurrencyOverrides {maxCount=2, resolutionRule=PreventNew, volumeScale=0.8}
}
